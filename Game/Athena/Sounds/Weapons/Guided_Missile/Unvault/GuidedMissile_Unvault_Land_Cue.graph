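# Sound cue for the unvaulted Guided Missile weapon's landing sound, using weapon-drop attenuation and the loot-collect sound class.
# A single wave player feeds the landing sample into a Modulator left at default settings, which slightly randomizes pitch and volume on each play.
# The cue sets a boosted volume multiplier.
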SoundCue /Game/Athena/Sounds/Weapons/Guided_Missile/Unvault/GuidedMissile_Unvault_Land_Cue
root=Modulator_0
nodes [
 {id=Modulator_0, children [WavePlayer_0]}
 {id=WavePlayer_0, properties {soundWaveAssetPtr=&/GuidedMissile_Unvault_Land_01}}
]
properties {
 attenuationSettings=&Fort_Weapon_Drop_Att
 maxDistance=3000
 soundClassObject=&Fort_Loot_Collect
 volumeMultiplier=1.5
}
